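# Physics asset for the AnimalJackets backpack cosmetic, defining collision bodies and joint limits for its dynamic animal-shaped parts.
# A capsule on backpack_root anchors sphere bodies for the head, belly, arms and legs, with three-segment ear chains hanging off the head.
# Angular limits keep the head, belly and most ear joints fairly stiff while allowing wider swing in the arms and legs.
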PhysicsAsset /BRCosmetics/Accessories/FORT_Backpacks/F_MED_AnimalJackets_Backpack/Meshes/F_MED_AnimalJackets_Backpack_Physics
bodies {
 backpack_root [{shapeType=Capsule, length=0.25, radius=10.6229, center=<-10.849 -7.7016 0.2249>, rotation=<rot 0 87.3042 -89.138>}]
 dyn_arm_01_l [{shapeType=Sphere, radius=1.9292, center=<0 -0.8186 0>}]
 dyn_arm_01_r [{shapeType=Sphere, radius=2.077, center=<0 -0.8199 0>}]
 dyn_arm_02_l [{shapeType=Sphere, radius=2.7694, center=<1.2674 -0.5367 -0.0997>}]
 dyn_arm_02_r [{shapeType=Sphere, radius=2.9097, center=<1.4582 -0.6082 0.0004>}]
 dyn_belly [{shapeType=Sphere, radius=5.0174, center=<0 0 -2>}]
 dyn_ear_01_l [{shapeType=Sphere, radius=2.2945, center=<0 -0.2638 0>}]
 dyn_ear_01_r [{shapeType=Sphere, radius=3.2441, center=<0 -0.055 0>}]
 dyn_ear_02_l [{shapeType=Sphere, radius=1.7502, center=<1.8999 0.0001 0>}]
 dyn_ear_02_r [{shapeType=Sphere, radius=1.9642, center=<1.2376 0 0>}]
 dyn_ear_03_l [{shapeType=Sphere, radius=1.354, center=<2.2413 0 0>}]
 dyn_ear_03_r [{shapeType=Sphere, radius=1.8756, center=<1.8403 0 -0.0001>}]
 dyn_head [{shapeType=Sphere, radius=6.4464, center=<3.7187 -0.2685 -0.2152>}]
 dyn_leg_01_l [{shapeType=Sphere, radius=3.2321, center=<1.1736 -0.0161 0.105>}]
 dyn_leg_01_r [{shapeType=Sphere, radius=3.2226, center=<1.1929 -0.0238 -0.1217>}]
 dyn_leg_02_l [{shapeType=Sphere, radius=3.3963, center=<1.9882 0.1101 -0.4347>}]
 dyn_leg_02_r [{shapeType=Sphere, radius=3.379, center=<2.0446 0.1147 0.3973>}]
}
constraints [
 {bone1Name=dyn_head, bone2Name=backpack_root, swing1LimitDegrees=15, swing2LimitDegrees=15, twistLimitDegrees=2}
 {bone1Name=dyn_ear_01_r, bone2Name=dyn_head, swing1LimitDegrees=5, swing2LimitDegrees=5, twistLimitDegrees=5}
 {bone1Name=dyn_ear_01_l, bone2Name=dyn_head, swing1LimitDegrees=5, swing2LimitDegrees=5, twistLimitDegrees=5}
 {bone1Name=dyn_ear_02_r, bone2Name=dyn_ear_01_r, swing1LimitDegrees=5, swing2LimitDegrees=5, twistLimitDegrees=5}
 {bone1Name=dyn_ear_02_l, bone2Name=dyn_ear_01_l, swing1LimitDegrees=8, swing2LimitDegrees=8, twistLimitDegrees=5}
 {bone1Name=dyn_ear_03_r, bone2Name=dyn_ear_02_r, swing1LimitDegrees=5, swing2LimitDegrees=5, twistLimitDegrees=5}
 {bone1Name=dyn_ear_03_l, bone2Name=dyn_ear_02_l, swing1LimitDegrees=20, swing2LimitDegrees=20, twistLimitDegrees=5}
 {bone1Name=dyn_arm_01_r, bone2Name=backpack_root, swing1LimitDegrees=45, swing2LimitDegrees=45, twistLimitDegrees=10}
 {bone1Name=dyn_arm_01_l, bone2Name=backpack_root, swing1LimitDegrees=45, swing2LimitDegrees=45, twistLimitDegrees=10}
 {bone1Name=dyn_arm_02_r, bone2Name=dyn_arm_01_r, swing1LimitDegrees=30, swing2LimitDegrees=30, twistLimitDegrees=10}
 {bone1Name=dyn_arm_02_l, bone2Name=dyn_arm_01_l, swing1LimitDegrees=30, swing2LimitDegrees=30, twistLimitDegrees=10}
 {bone1Name=dyn_leg_01_r, bone2Name=backpack_root, swing1LimitDegrees=25, swing2LimitDegrees=25, twistLimitDegrees=10}
 {bone1Name=dyn_leg_02_r, bone2Name=dyn_leg_01_r, swing1LimitDegrees=45, swing2LimitDegrees=45, twistLimitDegrees=10}
 {bone1Name=dyn_leg_01_l, bone2Name=backpack_root, swing1LimitDegrees=25, swing2LimitDegrees=25, twistLimitDegrees=10}
 {bone1Name=dyn_leg_02_l, bone2Name=dyn_leg_01_l, swing1LimitDegrees=45, swing2LimitDegrees=45, twistLimitDegrees=10}
 {bone1Name=dyn_belly, bone2Name=backpack_root, swing1LimitDegrees=10, swing2LimitDegrees=10, twistLimitDegrees=0}
]
