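# Animation Blueprint for the AnimalJackets backpack cosmetic, adding physics-driven secondary motion and an aiming pose adjustment.
# A RigidBody node simulates the backpack's physics asset in backpack_root space with overridden gravity and damped component motion.
# An additive bone transform then nudges and rolls backpack_root, with interpolated alpha from the targeting state.
# The event graph also derives backpack_rotate from owner spine bend, though the AnimGraph never reads it.
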AnimBlueprint /BRCosmetics/Accessories/FORT_Backpacks/F_MED_AnimalJackets_Backpack/Meshes/F_MED_AnimalJackets_Backpack_Skeleton_AnimBlueprint
target_skeleton=&/F_MED_AnimalJackets_Backpack_Skeleton
graphs [
 AnimGraph {nodes [
   AnimGraphNode_SubInput_0 = Input Pose()
   n1 = RigidBody(ComponentPose=Local To Component(LocalPose=@AnimGraphNode_SubInput_0)) {
    bOverrideWorldGravity=true
    overridePhysicsAsset=&/F_MED_AnimalJackets_Backpack_Physics
    simulationSpace=BaseBoneSpace
    alphaBoolBlend {blendOption=Linear}
    baseBoneRef {boneName=backpack_root}
    componentAppliedLinearAccClamp=<1000 1000 1000>
    componentLinearAccScale=<0.6 0.6 0.6>
    componentLinearVelScale=<0.4 0.4 0.4>
    overrideWorldGravity {z=-980}
   }
   n2 = "Transform (Modify) Bone - Bone: backpack_root"!(ComponentPose=@n1, Alpha=Get AimDownSight() -> (AimDownSight:double), Translation="0,1,3", Rotation="0,0,13.5") -> (Pose) {
    rotationMode=BMM_Additive
    rotationSpace=BCS_BoneSpace
    translationMode=BMM_Additive
    translationSpace=BCS_BoneSpace
    alphaBoolBlend {blendOption=Linear}
    alphaScaleBiasClamp {bInterpResult=true}
    boneToModify {boneName=backpack_root}
    rotation {roll=13.5}
    translation {y=1, z=3}
   }
   Output Pose(Result=Component To Local(ComponentPose=@n2)) {layerGroup=None}
  ]}
 EventGraph {nodes [
   Event BlueprintUpdateAnimation()
   > Set AimDownSight(AimDownSight:double="To Float (Boolean)"(InBool=Get bIsTargeting())) -> (Output_Get:double)
   n1 = GetOwnerSkeletalMeshComponent()
   n2 = BreakRotator(InRot=BreakTransform(InTransform=GetDeltaTransformFromRefPose(self=@n1, BoneName=spine_05, BaseName=spine_01)).Rotation)
   > Set backpack_rotate(backpack_rotate=MakeRotator(Yaw=MapRangeClamped(Value=@n2.Yaw, InRangeA=-5, InRangeB=-30, OutRangeB=15)))
   TryGetPawnOwner()
  ]}
]
variables [
 AimDownSight {type=number}
 backpack_rotate {type=Rotator}
]
